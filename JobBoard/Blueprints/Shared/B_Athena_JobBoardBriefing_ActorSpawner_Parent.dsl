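; Parent spawner actor for Job Board briefings that spawns configured actor classes and counts at EQS-chosen locations.
; On authority BeginPlay it runs the query template on the owning player's pawn, then checks there are enough results for the map's total.
; Each spawn consumes a random location, gets random yaw and can be marked through the owner's marker component.
; After a delay, bLastActorSpawned is set, whose RepNotify gives the spawner a short lifespan.
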(event EventBeginPlay
  ;[0 0 0] Begin play - run EQS
  (|CallParentFunction)
  (Networking|SwitchHasAuthority
    (:Authority
      ;;[0.2016 0.2016 0.2016] For instigator - run EQS query of provided template to determine spawn location(s)
      (bind _asfort_player_controller (Utilities|Casting|CastToFortPlayerController (Actor|GetOwner)))
      (bind _output_get (Variables|Hidden|SetOwningPlayerController _asfort_player_controller))
      (Utilities|IsValid _output_get
        (:Is%20Valid
          (Utilities|IsValid (Game|Player|GetPlayerPawn _output_get)
            (:Is%20Valid
              (bind _output_get_2 (Variables|Hidden|SetOwningPlayerPawn (Game|Player|GetPlayerPawn _output_get)))
              (bind _returnvalue (AI|EQS|RunEQSQuery (Variables|QueryConfig|GetQuery%20Template) _output_get_2 (Variables|QueryConfig|GetQueryRunMode)))
              (bind _output_get_3 (Variables|QueryConfig|SetTempEQS _returnvalue))
              (AI|EQS|BindEventtoOnQueryFinished _output_get_3 (EventDispatchers|CreateEvent :!event Custom|OnEQSQueryFinished)))))))))

(event Custom|OnEQSQueryFinished
  (params (QueryInstance :type EQS%20Query%20Instance%20Object%20Reference) (QueryStatus :type EEnvQueryStatus%20Enum :default "Processing"))
  ;[0 0 0] On EQS Query Finished - handle spawning the actors
  (bind (_resultlocations _returnvalue) (AI|EQS|GetQueryResultsasLocations QueryInstance))
  (Variables|Hidden|SetPlacementLocs _resultlocations)
  ;;[0.2016 0.2016 0.2016] Get the number of actors to spawn from the map and make sure there are enough locations to spawn them
  (if _returnvalue
    (Variables|Hidden|SetTotal%20Actor%20Count (CallFunction|GetTotalActorCounttoSpawn))
    (if (>= (Utilities|Array|Length (Variables|Hidden|GetPlacementLocs)) (Variables|Hidden|GetTotal%20Actor%20Count))
      ;;[0.2016 0.2016 0.2016] Start spawning actors from the actor map
      (if (Variables|SpawningConfig|GetbDelayFirstSpawn)
        (Utilities|FlowControl|Delay 1.5)
        (else
          _))
      (CallFunction|SpawnActorsfromMapEvent)
      (else
        _))
    (else
      _))
  (Fortnite|FortShippingLog self "EQS did not find enough locations to match desired actors to spawn count. Actors will not spawn and gameplay event will not be called" true))

(event EventEndPlay
  (params (EndPlayReason :type EEndPlayReason%20Enum :default "Destroyed"))
  (bind _tempeqs (Variables|QueryConfig|GetTempEQS))
  ;[0 0 0] End Play - remove EQS binds and cleanup any timers
  (|CallParentFunction EndPlayReason)
  (Utilities|IsValid _tempeqs
    (:Is%20Valid
      (AI|EQS|UnbindEventfromOnQueryFinished _tempeqs (EventDispatchers|CreateEvent :!event Custom|OnEQSQueryFinished)))))

(event Custom|ActorReadyToSpawn
  (bind _placementlocs (Variables|Hidden|GetPlacementLocs))
  (bind _returnvalue (Utilities|Array|Length _placementlocs))
  ;[0 0 0] Helper events
  ;;[1 1 1] When actor is ready to spawn, trigger any delays needed between spawns, then spawn the actor. Do not repeat locations already used
  (if (!= _returnvalue 0)
    (bind _output_get (Variables|SpawningConfig|SetCurrentArrayIndex (Math|Random|RandomIntegerinRange 0 (- _returnvalue 1))))
    (bind _spawnedactorreference (CallFunction|SpawnActor _output_get))
    (Utilities|Array|AddUnique (Variables|Hidden|GetSpawnedActors) _spawnedactorreference)
    (Utilities|Array|RemoveIndex (Variables|Hidden|GetPlacementLocs) (Variables|SpawningConfig|GetCurrentArrayIndex))))

(event Custom|SpawnActorsFromMap_Event
  ;[0 0 0] Helper events
  ;;[0.05729 0.05729 0.05729] For each actor class in the map, spawn the number of actors indicated for that class
  (bind (_keypin _valuepin) (Utilities|Map|ForEachLoop(Map) (Variables|SpawningConfig|GetActorsToSpawn))
    (:then
      (Variables|Hidden|SetSpawnActorClass _keypin)
      (for _ (range _valuepin)
        (CallFunction|ActorReadytoSpawn)))
    (:CompletedPin
      (Utilities|FlowControl|Delay 1.5)
      (Variables|Hidden|SetbLastActorSpawned true) ; only set if changed
    )))

(fn GetOwnerSquadmates
  ;[0 0 0] Get the squadmates present in the owner's squad. Inclues self, so array should return at least 1
  (bind _returnvalue (Fortnite|Utilities|AttemptGetPlayerStateforPawn (Variables|Hidden|GetOwningPlayerPawn)))
  (bind _asfort_player_state_athena (Utilities|Casting|CastToFortPlayerStateAthena _returnvalue))
  (bind _returnvalue_2 (Class|FortGameStateAthena|GetSquadMemberPlayerStates (Fortnite|Athena|GameState|GetGameStateAthenafromContext (Variables|Hidden|GetOwningPlayerController)) _asfort_player_state_athena true))
  (return _returnvalue_2))

(fn GetTotalActorCountToSpawn
  ;[0 0 0] For each integer value per class in the map, add the value ints together and return the total count
  (bind _values (Utilities|Map|Values (Variables|SpawningConfig|GetActorsToSpawn)))
  (for (_array_element _array_index) _values
    (if (== _array_index 0)
      (bind _output_get (LocalVariables|SetCurrActorCount _array_element))
      (LocalVariables|SetPrevActorCount _output_get)
      (else
        (bind _output_get_2 (LocalVariables|SetCurrActorCount _array_element))
        (LocalVariables|SetPrevActorCount (+ _output_get_2 (LocalVariables|GetPrevActorCount))))))
  (return (LocalVariables|GetPrevActorCount)))

(fn MarkActorForOwnerSquad
  (params (SpawnedActor :type Actor%20Object%20Reference))
  (bind _owningplayercontroller (Variables|Hidden|GetOwningPlayerController))
  (bind _returnvalue (Class|FortPlayerController|GetMarkerComponent _owningplayercontroller))
  ;[0 0 0] When the actor is spawned, mark it for the instigator and squadmates
  (Utilities|IsValid _owningplayercontroller
    (:Is%20Valid
      (Utilities|IsValid _returnvalue
        (:Is%20Valid
          (Utilities|IsValid SpawnedActor
            (:Is%20Valid
              (MarkerComponent|MarkActoronServer _returnvalue SpawnedActor true false))))))))

(fn OnRep_bLastActorSpawned
  ;[0 0 0] When the last actor is spawned, send a gameplay event to the owner squad and mark this spawner actor for deletion
  (Networking|SwitchHasAuthority
    (:Authority
      (if (not (Utilities|Array|IsNotEmpty (Variables|Hidden|GetSpawnedActors)))
        (Fortnite|FortShippingLog self "No actors spawned by briefing" true))
      (Actor|SetLifeSpan 0.5))))

(fn SpawnActor
  (params (LocationIndex :type Integer) (bLastActorToSpawn :type Boolean))
  (bind _spawnactorclass (Variables|Hidden|GetSpawnActorClass))
  ;[0 0 0] Spawn an actor of the indicated class at the specified location index from the returned EQS query
  (if (Utilities|IsValidClass _spawnactorclass)
    (bind _returnvalue (Game|SpawnActorfromClass _spawnactorclass (Math|Transform|MakeTransform (Utilities|Array|Get(acopy) (Variables|Hidden|GetPlacementLocs) LocationIndex) (Math|Rotator|MakeRotator 0.0 0.0 (Math|Random|RandomFloatinRange 0.0 3.6e+02)) "1.000000,1.000000,1.000000") "AdjustIfPossibleButAlwaysSpawn" "SelectDefaultAtRuntime" (Actor|GetOwner) 0))
    (if (Variables|SpawningConfig|GetbMarkActorsOnSpawn)
      (CallFunction|MarkActorforOwnerSquad _returnvalue)
      (else
        _))
    (return _returnvalue)))

(fn ConstructionScript)
